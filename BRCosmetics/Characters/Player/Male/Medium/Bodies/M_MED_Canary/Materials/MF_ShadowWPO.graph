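# Material function for the Canary character body that alters world position offset in the shadow pass only, leaving the visible mesh unchanged.
# It shifts shadow-casting vertices along the global sunlight direction by ShadowOffset, then blends toward collapsing them onto the object position using ShadowCrunch.
# ShadowReplace feeds the result back into the incoming material attributes.
# A red BaseColor blend driven by ShadowCrunch is left unconnected.
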MaterialFunction /BRCosmetics/Characters/Player/Male/Medium/Bodies/M_MED_Canary/Materials/MF_ShadowWPO
graph {nodes [
  n1 = FunctionInput() {inputType=FunctionInput_MaterialAttributes}
  n2 = GetMaterialAttributes(None=@n1) {attributeGetTypes [#1, #2]}
  n3 = CollectionParameter() {
   collection=&FortniteMaterialParameters
   parameterName=SunlightDirectionalVector
  }
  n4 = ComponentMask(None=@n3) {b=true, g=true, r=true}
  n5 = ScalarParameter() {group=ShadowWPO, parameterName=ShadowOffset, sortPriority=0}
  n6 = WorldPosition()
  n7 = ObjectPositionWS()
  n8 = ScalarParameter() {group=ShadowWPO, parameterName=ShadowCrunch}
  n9 = LinearInterpolate(A=Add(A=@n2.WorldPositionOffset, B=Multiply(A=@n4, B=@n5)), B=Multiply(A=Subtract(A=@n6.XYZ, B=@n7)) {constB=-1}, Alpha=@n8)
  FunctionOutput(None=SetMaterialAttributes!(MaterialAttributes=@n1, WorldPositionOffset=ShadowReplace(Default=@n2.WorldPositionOffset, Shadow=@n9), BaseColor) {attributeSetTypes [#1, #2]})
  LinearInterpolate(A=@n2.BaseColor, B=Constant3Vector() {constant=<rgba 1 0 0 0>}, Alpha=@n8)
 ]}
