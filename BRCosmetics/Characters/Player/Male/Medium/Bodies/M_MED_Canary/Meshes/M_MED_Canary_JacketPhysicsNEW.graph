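# Physics asset for the Canary male medium body's jacket, giving its dynamic bones capsules and constraints for cloth-like motion.
# Seven chains (front, under-arm and back pairs, plus mid back) hang from small master-bone capsules attached to spine_04, mostly with narrow swing limits.
# Pelvis, spine and thigh capsules form the body volume, joined by free angular constraints.
# An extra swing-locked constraint ties the right under-arm tip to the right front-torso tip, with no left counterpart.
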PhysicsAsset /BRCosmetics/Characters/Player/Male/Medium/Bodies/M_MED_Canary/Meshes/M_MED_Canary_JacketPhysicsNEW
bodies {
 dyn_back_01_l [{shapeType=Capsule, length=5.5, radius=4.5, shapeName=dyn_back_01_l_capsule, center=<5 5 0>}]
 dyn_back_01_r [{shapeType=Capsule, length=6, radius=4, shapeName=dyn_back_01_r_capsule, center=<-5 -5 0>}]
 dyn_back_02_l [{shapeType=Capsule, length=4.75, radius=5.25, shapeName=dyn_back_02_l_capsule, center=<5 5 0>}]
 dyn_back_02_r [{shapeType=Capsule, length=4.5, radius=5.5, shapeName=dyn_back_02_r_capsule, center=<-5 -5 0>}]
 dyn_back_03_l [{shapeType=Capsule, length=10.5, radius=4, shapeName=dyn_back_03_l_capsule, center=<5 5 0>}]
 dyn_back_03_r [{shapeType=Capsule, length=10.5, radius=4, shapeName=dyn_back_03_r_capsule, center=<-5 -5 0>}]
 dyn_mid_back_01 [{shapeType=Capsule, length=6.5, radius=3.5, shapeName=dyn_mid_back_01_capsule, center=<6.2031 -5 0>}]
 dyn_mid_back_02 [{shapeType=Capsule, length=4.75, radius=5.25, shapeName=dyn_mid_back_02_capsule, center=<5 -5 0>}]
 dyn_mid_back_03 [{shapeType=Capsule, length=8.25, radius=6.25, shapeName=dyn_mid_back_03_capsule, center=<5 -9 0>}]
 dyn_torso_00_l [{shapeType=Capsule, length=0, radius=6.75, shapeName=dyn_torso_01_l_capsule, center=<5 6 5>}]
 dyn_torso_00_r [{shapeType=Capsule, length=4, radius=6, shapeName=dyn_torso_01_r_capsule, center=<-5 -6 0>}]
 dyn_torso_01_l [{shapeType=Capsule, length=0, radius=6.75, shapeName=dyn_torso_01_l_capsule, center=<5 6 5>}]
 dyn_torso_01_r [{shapeType=Capsule, length=9.7905, radius=6, shapeName=dyn_torso_01_r_capsule, center=<-5 -6 0>}]
 dyn_torso_02_l [{shapeType=Capsule, length=5, radius=6, shapeName=dyn_torso_02_l_capsule, center=<5 6 5>, rotation=<rot 90 0 -0.0002>}]
 dyn_torso_02_r [{shapeType=Capsule, length=10.8184, radius=5.75, shapeName=dyn_torso_02_r_capsule, center=<-5 -6 0>}]
 dyn_torso_03_l [{shapeType=Capsule, length=5, radius=6, shapeName=dyn_torso_03_l_capsule, center=<5 6 5>, rotation=<rot 90 0 -0.0002>}]
 dyn_torso_03_r [{shapeType=Capsule, length=6.8159, radius=5, shapeName=dyn_torso_03_r_capsule, center=<-5 -6 0>}]
 dyn_under_arm_01_l [{shapeType=Capsule, length=5, radius=5, shapeName=dyn_under_arm_01_l_capsule, center=<5 4 0>}]
 dyn_under_arm_01_r [{shapeType=Capsule, length=5, radius=5, shapeName=dyn_under_arm_01_r_capsule, center=<-5 -4 0>}]
 dyn_under_arm_02_l [{shapeType=Capsule, length=5, radius=5, shapeName=dyn_under_arm_02_l_capsule, center=<5 4 0>}]
 dyn_under_arm_02_r [{shapeType=Capsule, length=5, radius=5, shapeName=dyn_under_arm_02_r_capsule, center=<-5 -4 0>}]
 dyn_under_arm_03_l [{shapeType=Capsule, length=5, radius=5, shapeName=dyn_under_arm_03_l_capsule, center=<5 4 0>}]
 dyn_under_arm_03_r [{shapeType=Capsule, length=11, radius=5, shapeName=dyn_under_arm_03_r_capsule, center=<-5 -4 0>}]
 mstr_back_l [{shapeType=Capsule, length=0.505, radius=0.505, shapeName=mstr_back_l_capsule, rotation=<rot 0 -90 0>}]
 mstr_back_r [{shapeType=Capsule, length=0.505, radius=0.505, shapeName=mstr_back_r_capsule, rotation=<rot 0 -90 0>}]
 mstr_frnt_torso_l [{shapeType=Capsule, length=0.505, radius=0.505, shapeName=mstr_frnt_torso_l_capsule, rotation=<rot 0 -90 0>}]
 mstr_frnt_torso_r [{shapeType=Capsule, length=0.505, radius=0.505, shapeName=mstr_frnt_torso_r_capsule, rotation=<rot 0 -90 0>}]
 mstr_mid_back [{shapeType=Capsule, length=0.505, radius=0.505, shapeName=mstr_mid_back_capsule, rotation=<rot 0 -90 0>}]
 mstr_under_arm_l [{shapeType=Capsule, length=0.505, radius=0.505, shapeName=mstr_under_arm_l_capsule, rotation=<rot 0 -90 0>}]
 mstr_under_arm_r [{shapeType=Capsule, length=0.505, radius=0.505, shapeName=mstr_under_arm_r_capsule, rotation=<rot 0 -90 0>}]
 pelvis [
  {shapeType=Capsule, length=9.5, radius=12.75, shapeName=pelvis_capsule, center=<-7.7551 -1.6679 0>}
  {shapeType=Capsule, length=24.75, radius=15, shapeName=pelvis_capsule, center=<-18.158 -5.3861 0>}
 ]
 spine_01 [{shapeType=Capsule, length=7.75, radius=13, shapeName=spine_01_capsule}]
 spine_02 [{shapeType=Capsule, length=4.75, radius=13, shapeName=spine_02_capsule}]
 spine_03 [{shapeType=Capsule, length=4.25, radius=13.5, shapeName=spine_03_capsule}]
 spine_04 [{shapeType=Capsule, length=3.75, radius=14, shapeName=spine_04_capsule}]
 thigh_l [{shapeType=Capsule, length=41.25, radius=9.5, shapeName=thigh_l_capsule, center=<-24.6034 0.2818 0>, rotation=<rot -90 180 179.9989>}]
 thigh_r [{
   shapeType=Capsule
   length=38.9477
   radius=11.8023
   shapeName=thigh_r_capsule
   center=<24.7953 -0.0003 1.2941>
   rotation=<rot -90 180 -175.0007>
  }]
}
constraints [
 {
  bone1Name=spine_01
  bone2Name=pelvis
  swing1LimitDegrees=45
  swing1Motion=Free
  swing2LimitDegrees=45
  swing2Motion=Free
  twistLimitDegrees=45
  twistMotion=Free
 }
 {
  bone1Name=spine_02
  bone2Name=spine_01
  swing1LimitDegrees=45
  swing1Motion=Free
  swing2LimitDegrees=45
  swing2Motion=Free
  twistLimitDegrees=45
  twistMotion=Free
 }
 {
  bone1Name=spine_03
  bone2Name=spine_02
  swing1LimitDegrees=45
  swing1Motion=Free
  swing2LimitDegrees=45
  swing2Motion=Free
  twistLimitDegrees=45
  twistMotion=Free
 }
 {
  bone1Name=spine_04
  bone2Name=spine_03
  swing1LimitDegrees=45
  swing1Motion=Free
  swing2LimitDegrees=45
  swing2Motion=Free
  twistLimitDegrees=45
  twistMotion=Free
 }
 {
  bone1Name=mstr_frnt_torso_r
  bone2Name=spine_04
  swing1LimitDegrees=45
  swing1Motion=Free
  swing2LimitDegrees=45
  swing2Motion=Free
  twistLimitDegrees=45
  twistMotion=Free
 }
 {bone1Name=dyn_torso_01_r, bone2Name=dyn_torso_00_r, swing1LimitDegrees=10, swing2LimitDegrees=10, twistLimitDegrees=0}
 {bone1Name=dyn_torso_02_r, bone2Name=dyn_torso_01_r, swing1LimitDegrees=10, swing2LimitDegrees=10, twistLimitDegrees=0}
 {bone1Name=dyn_torso_03_r, bone2Name=dyn_torso_02_r, swing1LimitDegrees=10, swing2LimitDegrees=10, twistLimitDegrees=0}
 {
  bone1Name=mstr_under_arm_r
  bone2Name=spine_04
  swing1LimitDegrees=45
  swing1Motion=Free
  swing2LimitDegrees=45
  swing2Motion=Free
  twistLimitDegrees=45
  twistMotion=Free
 }
 {bone1Name=dyn_under_arm_01_r, bone2Name=mstr_under_arm_r, swing1LimitDegrees=5, swing2LimitDegrees=5, twistLimitDegrees=0}
 {bone1Name=dyn_under_arm_02_r, bone2Name=dyn_under_arm_01_r, swing1LimitDegrees=10, swing2LimitDegrees=10, twistLimitDegrees=0}
 {bone1Name=dyn_under_arm_03_r, bone2Name=dyn_under_arm_02_r, swing1LimitDegrees=10, swing2LimitDegrees=10, twistLimitDegrees=0}
 {
  bone1Name=mstr_back_r
  bone2Name=spine_04
  swing1LimitDegrees=45
  swing1Motion=Free
  swing2LimitDegrees=45
  swing2Motion=Free
  twistLimitDegrees=45
  twistMotion=Free
 }
 {bone1Name=dyn_back_01_r, bone2Name=mstr_back_r, swing1LimitDegrees=10, swing2LimitDegrees=5, twistLimitDegrees=0}
 {bone1Name=dyn_back_02_r, bone2Name=dyn_back_01_r, swing1LimitDegrees=15, swing2LimitDegrees=5, twistLimitDegrees=0}
 {bone1Name=dyn_back_03_r, bone2Name=dyn_back_02_r, swing1LimitDegrees=20, swing2LimitDegrees=5, twistLimitDegrees=0}
 {
  bone1Name=mstr_back_l
  bone2Name=spine_04
  swing1LimitDegrees=45
  swing1Motion=Free
  swing2LimitDegrees=45
  swing2Motion=Free
  twistLimitDegrees=45
  twistMotion=Free
 }
 {bone1Name=dyn_back_01_l, bone2Name=mstr_back_l, swing1LimitDegrees=10, swing2LimitDegrees=5, twistLimitDegrees=0}
 {bone1Name=dyn_back_02_l, bone2Name=dyn_back_01_l, swing1LimitDegrees=15, swing2LimitDegrees=5, twistLimitDegrees=0}
 {bone1Name=dyn_back_03_l, bone2Name=dyn_back_02_l, swing1LimitDegrees=20, swing2LimitDegrees=5, twistLimitDegrees=0}
 {
  bone1Name=mstr_frnt_torso_l
  bone2Name=spine_04
  swing1LimitDegrees=45
  swing1Motion=Free
  swing2LimitDegrees=45
  swing2Motion=Free
  twistLimitDegrees=45
  twistMotion=Free
 }
 {bone1Name=dyn_torso_01_l, bone2Name=dyn_torso_00_l, swing1LimitDegrees=5, swing2LimitDegrees=20, twistLimitDegrees=0}
 {bone1Name=dyn_torso_02_l, bone2Name=dyn_torso_01_l, swing1LimitDegrees=5, swing2LimitDegrees=20, twistLimitDegrees=0}
 {bone1Name=dyn_torso_03_l, bone2Name=dyn_torso_02_l, swing1LimitDegrees=5, swing2LimitDegrees=20, twistLimitDegrees=0}
 {
  bone1Name=mstr_mid_back
  bone2Name=spine_04
  swing1LimitDegrees=45
  swing1Motion=Free
  swing2LimitDegrees=45
  swing2Motion=Free
  twistLimitDegrees=45
  twistMotion=Free
 }
 {bone1Name=dyn_mid_back_01, bone2Name=mstr_mid_back, swing1LimitDegrees=10, swing2LimitDegrees=5, twistLimitDegrees=0}
 {bone1Name=dyn_mid_back_02, bone2Name=dyn_mid_back_01, swing1LimitDegrees=15, swing2LimitDegrees=5, twistLimitDegrees=0}
 {bone1Name=dyn_mid_back_03, bone2Name=dyn_mid_back_02, swing1LimitDegrees=20, swing2LimitDegrees=5, twistLimitDegrees=0}
 {
  bone1Name=mstr_under_arm_l
  bone2Name=spine_04
  swing1LimitDegrees=45
  swing1Motion=Free
  swing2LimitDegrees=45
  swing2Motion=Free
  twistLimitDegrees=45
  twistMotion=Free
 }
 {bone1Name=dyn_under_arm_01_l, bone2Name=mstr_under_arm_l, swing1LimitDegrees=5, swing2LimitDegrees=5, twistLimitDegrees=0}
 {bone1Name=dyn_under_arm_02_l, bone2Name=dyn_under_arm_01_l, swing1LimitDegrees=5, swing2LimitDegrees=5, twistLimitDegrees=0}
 {bone1Name=dyn_under_arm_03_l, bone2Name=dyn_under_arm_02_l, swing1LimitDegrees=5, swing2LimitDegrees=5, twistLimitDegrees=0}
 {
  bone1Name=thigh_l
  bone2Name=pelvis
  swing1LimitDegrees=45
  swing1Motion=Free
  swing2LimitDegrees=45
  swing2Motion=Free
  twistLimitDegrees=45
  twistMotion=Free
 }
 {
  bone1Name=thigh_r
  bone2Name=pelvis
  swing1LimitDegrees=45
  swing1Motion=Free
  swing2LimitDegrees=45
  swing2Motion=Free
  twistLimitDegrees=45
  twistMotion=Free
 }
 {
  bone1Name=dyn_torso_00_r
  bone2Name=mstr_frnt_torso_r
  swing1LimitDegrees=45
  swing1Motion=Free
  swing2LimitDegrees=45
  swing2Motion=Free
  twistLimitDegrees=45
  twistMotion=Free
 }
 {bone1Name=dyn_torso_00_l, bone2Name=mstr_frnt_torso_l, swing1LimitDegrees=45, swing2LimitDegrees=45, twistLimitDegrees=45, twistMotion=Free}
 {
  bone1Name=dyn_under_arm_03_r
  bone2Name=dyn_torso_03_r
  swing1LimitDegrees=45
  swing1Motion=Locked
  swing2LimitDegrees=45
  swing2Motion=Locked
  twistLimitDegrees=45
  twistMotion=Free
 }
]
