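# Physics asset for the rope and bag dangling accessories on the medium male Canary body.
# Rope (main plus three dangly segments) and two-segment bag chains hang from master bones constrained to the pelvis, with tight swing and zero twist limits.
# Pelvis, left thigh and left calf capsules, linked by free angular constraints, provide body shapes.
# A huge zero-length root capsule, offset downward, apparently approximates the ground.
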PhysicsAsset /BRCosmetics/Characters/Player/Male/Medium/Bodies/M_MED_Canary/Meshes/M_MED_Canary_RopeBagPhysics
bodies {
 calf_l [{shapeType=Capsule, length=32.7668, radius=9.5957, shapeName=calf_l_capsule, center=<-32.0481 0 0>, rotation=<rot 90 0 0>}]
 dyn_bag_Main [{shapeType=Capsule, length=8.5, radius=4.75, shapeName=dyn_bag_Main_capsule, center=<8 0 0>, rotation=<rot 0 180 180>}]
 dyn_bag_Main2 [{shapeType=Capsule, length=13, radius=4.25, shapeName=dyn_bag_Main_capsule, center=<8 0 0>, rotation=<rot 0 180 180>}]
 dyn_ropeDangly_01 [{shapeType=Capsule, length=5.25, radius=1.5, shapeName=dyn_ropeDangly_01_capsule, center=<8.2514 0 0>, rotation=<rot -90 180 180>}]
 dyn_ropeDangly_02 [{shapeType=Capsule, length=5, radius=1.75, shapeName=dyn_ropeDangly_02_capsule, center=<5 0 0>, rotation=<rot -90 0 0>}]
 dyn_ropeDangly_03 [{shapeType=Capsule, length=5, radius=1.75, shapeName=dyn_ropeDangly_03_capsule, center=<5 0 0>, rotation=<rot -90 0 0>}]
 dyn_ropeMain_01 [{shapeType=Capsule, length=21.0896, radius=6.5144, shapeName=mstr_ropeMain_capsule, center=<15 5 0>, rotation=<rot -0.0001 0 5>}]
 mstr_bag_Main [{shapeType=Capsule, length=0.505, radius=0.505, shapeName=mstr_bag_Main_capsule, rotation=<rot 0 -90 0>}]
 mstr_ropeMain [{shapeType=Capsule, length=22.8396, radius=4.7644, shapeName=mstr_ropeMain_capsule, center=<5 5 0>}]
 pelvis [{shapeType=Capsule, length=0, radius=10.5799, shapeName=pelvis_capsule, center=<-9.9014 0 0>}]
 root [{shapeType=Capsule, length=0, radius=500, shapeName=root_capsule, center=<51.821 0 -495.1773>, rotation=<rot 0 -90 0>}]
 thigh_l [{shapeType=Capsule, length=22.9289, radius=9.6431, shapeName=thigh_l_capsule, center=<-18.0189 6.6693 0>, rotation=<rot 90 0 0>}]
}
constraints [
 {bone1Name=dyn_ropeDangly_01, bone2Name=dyn_ropeMain_01, swing1LimitDegrees=20, swing2LimitDegrees=20, twistLimitDegrees=0}
 {bone1Name=dyn_ropeDangly_02, bone2Name=dyn_ropeDangly_01, swing1LimitDegrees=20, swing2LimitDegrees=20, twistLimitDegrees=0}
 {bone1Name=dyn_ropeDangly_03, bone2Name=dyn_ropeDangly_02, swing1LimitDegrees=20, swing2LimitDegrees=20, twistLimitDegrees=0}
 {bone1Name=dyn_ropeMain_01, bone2Name=mstr_ropeMain, swing1LimitDegrees=20, swing2LimitDegrees=20, twistLimitDegrees=0}
 {
  bone1Name=thigh_l
  bone2Name=pelvis
  swing1LimitDegrees=45
  swing1Motion=Free
  swing2LimitDegrees=45
  swing2Motion=Free
  twistLimitDegrees=45
  twistMotion=Free
 }
 {
  bone1Name=mstr_bag_Main
  bone2Name=pelvis
  swing1LimitDegrees=45
  swing1Motion=Free
  swing2LimitDegrees=45
  swing2Motion=Free
  twistLimitDegrees=45
  twistMotion=Free
 }
 {bone1Name=mstr_ropeMain, bone2Name=pelvis, swing1LimitDegrees=20, swing2LimitDegrees=20, twistLimitDegrees=0}
 {bone1Name=dyn_bag_Main2, bone2Name=dyn_bag_Main, swing1LimitDegrees=10, swing2LimitDegrees=10, twistLimitDegrees=0}
 {bone1Name=dyn_bag_Main, bone2Name=mstr_bag_Main, swing1LimitDegrees=20, swing2LimitDegrees=20, twistLimitDegrees=0}
 {
  bone1Name=pelvis
  bone2Name=root
  swing1LimitDegrees=45
  swing1Motion=Free
  swing2LimitDegrees=45
  swing2Motion=Free
  twistLimitDegrees=45
  twistMotion=Free
 }
 {
  bone1Name=calf_l
  bone2Name=thigh_l
  swing1LimitDegrees=45
  swing1Motion=Free
  swing2LimitDegrees=45
  swing2Motion=Free
  twistLimitDegrees=45
  twistMotion=Free
 }
]
