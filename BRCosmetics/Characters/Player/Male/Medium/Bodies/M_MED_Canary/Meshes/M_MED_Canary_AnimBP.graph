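# Post-pose AnimBlueprint for the M_MED_Canary player body cosmetic, correcting and simulating its jacket over the attached parent's pose.
# Copy Pose From Mesh feeds a deform Control Rig, pistol bone constraint and RigidBody jacket simulation relative to spine_05.
# Additive bone tweaks, weighted by movement and limb-pose variables like crouch, sprint, jump and skydive, adjust the jacket and upper arms.
# A near-duplicate chain uses another jacket physics asset.
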AnimBlueprint /BRCosmetics/Characters/Player/Male/Medium/Bodies/M_MED_Canary/Meshes/M_MED_Canary_AnimBP
target_skeleton=&/M_MED_Canary_Skeleton
graphs [
 AnimGraph {nodes [
   Local To Component()
   "Transform (Modify) Bone - Bone: dyn_torso_01_r"(Alpha=1, Translation="0,0,0", Rotation="-30,0,0", Scale="1,1,1") {
    rotationMode=BMM_Additive
    rotationSpace=BCS_BoneSpace
    scaleMode=BMM_Additive
    translationMode=BMM_Additive
    alphaBoolBlend {blendOption=Linear}
    alphaScaleBiasClamp {bInterpResult=true, interpSpeedDecreasing=0.5, interpSpeedIncreasing=2}
    boneToModify {boneName=dyn_torso_01_r}
    rotation {pitch=-30}
   }
   n1 = Control Rig(Source=Copy Pose From Mesh() {bUseAttachedParent=true}) {
    alphaBoolBlend {blendOption=Linear}
    controlRigAssetReference {blueprintRigClass=&/M_MED_Canary_DeformRig_Master_ControlRig1.M_MED_Canary_DeformRig_Master_ControlRig1_C}
   }
   n2 = "Transform (Modify) Bone - Bone: dyn_torso_00_l"(ComponentPose=Local To Component(LocalPose=@n1), Alpha=1, Translation="0,0,0", Rotation="15,0,0", Scale="1,1,1") {
    rotationMode=BMM_Additive
    rotationSpace=BCS_BoneSpace
    scaleMode=BMM_Additive
    translationMode=BMM_Additive
    alphaBoolBlend {blendOption=Linear}
    alphaScaleBiasClamp {bInterpResult=true, interpSpeedIncreasing=50}
    boneToModify {boneName=dyn_torso_00_l}
    rotation {pitch=15}
   }
   n3 = "Transform (Modify) Bone - Bone: dyn_torso_00_r"(ComponentPose=@n2, Alpha=1, Translation="0,0,0", Rotation="15,0,0", Scale="1,1,1") {
    rotationMode=BMM_Additive
    rotationSpace=BCS_BoneSpace
    scaleMode=BMM_Additive
    translationMode=BMM_Additive
    alphaBoolBlend {blendOption=Linear}
    alphaScaleBiasClamp {bInterpResult=true, interpSpeedIncreasing=50}
    boneToModify {boneName=dyn_torso_00_r}
    rotation {pitch=15}
   }
   n4 = "Constraint - pistol (PPPP)"!(ComponentPose=@n3, Alpha=1, ConstraintWeights_0=0, ConstraintWeights_1:float=0, ConstraintWeights_2:float=1, ConstraintWeights_3:float=0) -> (Pose) {
    alphaBoolBlend {blendOption=Linear}
    boneToModify {boneName=pistol}
    constraintSetup [
     {targetBone {boneName=pelvis}}
     {targetBone {boneName=thigh_r}}
     {targetBone {boneName=thigh_twist_01_r}}
     {targetBone {boneName=deform_groin_r}}
    ]
    constraintWeights [1,1,1,1]
   }
   n5 = "Transform (Modify) Bone - Bone: pistol"(ComponentPose=@n4, Alpha=1, Translation="0,0,0", Rotation="0,0,20", Scale="1,1,1") {
    rotationMode=BMM_Additive
    rotationSpace=BCS_BoneSpace
    scaleMode=BMM_Additive
    translationMode=BMM_Additive
    alphaBoolBlend {blendOption=Linear}
    alphaScaleBiasClamp {bInterpResult=true, interpSpeedIncreasing=50}
    boneToModify {boneName=pistol}
    rotation {roll=20}
   }
   n6 = "Transform (Modify) Bone - Bone: dyn_torso_00_l"(
    ComponentPose=@n5
    Alpha=Get IsCrouchingNoMovingThighUpL() -> (IsCrouchingNoMovingThighUpL:float)
    Translation="3,2,3"
    Rotation="20,-20,0"
    Scale="0.25,1,1"
   ) {
    rotationMode=BMM_Additive
    rotationSpace=BCS_BoneSpace
    scaleMode=BMM_Additive
    translationMode=BMM_Additive
    alphaBoolBlend {blendOption=Linear}
    alphaScaleBiasClamp {interpSpeedIncreasing=1000}
    boneToModify {boneName=dyn_torso_00_l}
    rotation {pitch=20, yaw=-20}
    scale {x=0.25}
    translation=<3 2 3>
   }
   n7 = Get ThighUpR() -> (ThighUpR:float)
   n8 = "Transform (Modify) Bone - Bone: mstr_under_arm_r"(ComponentPose=@n6, Alpha=@n7, Translation="0,0,0", Rotation="5,-8,0", Scale="1,1,1") {
    rotationMode=BMM_Additive
    rotationSpace=BCS_BoneSpace
    scaleMode=BMM_Additive
    translationMode=BMM_Additive
    alphaBoolBlend {blendOption=Linear}
    alphaScaleBiasClamp {bInterpResult=true, interpSpeedIncreasing=50}
    boneToModify {boneName=mstr_under_arm_r}
    rotation {pitch=5, yaw=-8}
   }
   n9 = "Transform (Modify) Bone - Bone: mstr_frnt_torso_r"(ComponentPose=@n8, Alpha=@n7, Translation="0,0,0", Rotation="5,-8,0", Scale="1,1,1") {
    rotationMode=BMM_Additive
    rotationSpace=BCS_BoneSpace
    scaleMode=BMM_Additive
    translationMode=BMM_Additive
    alphaBoolBlend {blendOption=Linear}
    alphaScaleBiasClamp {bInterpResult=true, interpSpeedIncreasing=50}
    boneToModify {boneName=mstr_frnt_torso_r}
    rotation {pitch=5, yaw=-8}
   }
   n10 = RigidBody!(ComponentPose=@n9, bAlphaBoolEnabled:bool, OverrideWorldGravity:Vector=Get GravityOverride(), ExternalForce) -> (Pose) {
    alphaInputType=Bool
    bClampLinearTranslationLimitToRefPose=true
    bOverrideWorldGravity=true
    overridePhysicsAsset=&/M_MED_Canary_JacketPhysicsNEW
    simulationSpace=BaseBoneSpace
    alphaBoolBlend {blendOption=Linear}
    baseBoneRef {boneName=spine_05}
    componentAppliedLinearAccClamp=<500 500 500>
    componentLinearAccScale=<0.15 0.25 0.25>
    componentLinearVelScale=<0.3 0.3 0.3>
    overrideWorldGravity {z=-980}
    simSpaceSettings {
     maxAngularAcceleration=0
     maxAngularVelocity=0.3
     maxLinearAcceleration=100
     maxLinearVelocity=100
     velocityScaleZ=0
     worldAlpha=1
     externalLinearDragV=<0.2 0.2>
    }
   }
   n11 = Get IsNotCrouching() -> (IsNotCrouching:float)
   n12 = "Transform (Modify) Bone - Bone: mstr_frnt_torso_r"(ComponentPose=@n10, Alpha=@n11, Translation="-2,0,2", Rotation="3,0,0", Scale="1,1,1") {
    rotationMode=BMM_Additive
    rotationSpace=BCS_BoneSpace
    scaleMode=BMM_Additive
    translationMode=BMM_Additive
    alphaBoolBlend {blendOption=Linear}
    alphaScaleBiasClamp {bInterpResult=true, interpSpeedIncreasing=50}
    boneToModify {boneName=mstr_frnt_torso_r}
    rotation {pitch=3}
    translation {x=-2, z=2}
   }
   n13 = "Transform (Modify) Bone - Bone: mstr_under_arm_r"(ComponentPose=@n12, Alpha=@n11, Translation="-2,0,2", Rotation="3,0,0", Scale="1,1,1") {
    rotationMode=BMM_Additive
    rotationSpace=BCS_BoneSpace
    scaleMode=BMM_Additive
    translationMode=BMM_Additive
    alphaBoolBlend {blendOption=Linear}
    alphaScaleBiasClamp {bInterpResult=true, interpSpeedIncreasing=50}
    boneToModify {boneName=mstr_under_arm_r}
    rotation {pitch=3}
    translation {x=-2, z=2}
   }
   n14 = Get IsCrouching() -> (IsCrouching:float)
   n15 = "Transform (Modify) Bone - Bone: mstr_frnt_torso_r"(ComponentPose=@n13, Alpha=@n14, Translation="-1,0,1", Rotation="3,0,0", Scale="1,1,1") {
    rotationMode=BMM_Additive
    rotationSpace=BCS_BoneSpace
    scaleMode=BMM_Additive
    translationMode=BMM_Additive
    alphaBoolBlend {blendOption=Linear}
    alphaScaleBiasClamp {bInterpResult=true, interpSpeedIncreasing=50}
    boneToModify {boneName=mstr_frnt_torso_r}
    rotation {pitch=3}
    translation {x=-1, z=1}
   }
   n16 = "Transform (Modify) Bone - Bone: mstr_under_arm_r"(ComponentPose=@n15, Alpha=@n14, Translation="-1,0,1", Rotation="3,0,0", Scale="1,1,1") {
    rotationMode=BMM_Additive
    rotationSpace=BCS_BoneSpace
    scaleMode=BMM_Additive
    translationMode=BMM_Additive
    alphaBoolBlend {blendOption=Linear}
    alphaScaleBiasClamp {bInterpResult=true, interpSpeedIncreasing=50}
    boneToModify {boneName=mstr_under_arm_r}
    rotation {pitch=3}
    translation {x=-1, z=1}
   }
   n17 = Get IsSprintingNoCrouch() -> (IsSprintingNoCrouch:float)
   n18 = "Transform (Modify) Bone - Bone: dyn_back_02_r"(ComponentPose=@n16, Alpha=@n17, Translation="0,0,0", Rotation="0,0,0", Scale="1,1,1") {
    rotationMode=BMM_Additive
    rotationSpace=BCS_BoneSpace
    scaleMode=BMM_Additive
    translationMode=BMM_Additive
    alphaBoolBlend {blendOption=Linear}
    alphaScaleBiasClamp {bInterpResult=true, interpSpeedDecreasing=500, interpSpeedIncreasing=1}
    boneToModify {boneName=dyn_back_02_r}
   }
   n19 = "Transform (Modify) Bone - Bone: dyn_back_02_l"(ComponentPose=@n18, Alpha=@n17, Translation="0,0,0", Rotation="0,-10,0", Scale="1,1,1") {
    rotationMode=BMM_Additive
    rotationSpace=BCS_BoneSpace
    scaleMode=BMM_Additive
    translationMode=BMM_Additive
    alphaBoolBlend {blendOption=Linear}
    alphaScaleBiasClamp {bInterpResult=true, interpSpeedDecreasing=500, interpSpeedIncreasing=1}
    boneToModify {boneName=dyn_back_02_l}
    rotation {yaw=-10}
   }
   n20 = "Transform (Modify) Bone - Bone: dyn_mid_back_02"(ComponentPose=@n19, Alpha=@n17, Translation="0,0,0", Rotation="0,10,0", Scale="1,1,1") {
    rotationMode=BMM_Additive
    rotationSpace=BCS_BoneSpace
    scaleMode=BMM_Additive
    translationMode=BMM_Additive
    alphaBoolBlend {blendOption=Linear}
    alphaScaleBiasClamp {bInterpResult=true, interpSpeedDecreasing=500, interpSpeedIncreasing=1}
    boneToModify {boneName=dyn_mid_back_02}
    rotation {yaw=10}
   }
   n21 = "Transform (Modify) Bone - Bone: mstr_frnt_torso_r"(ComponentPose=@n20, Alpha=@n17, Translation="3,0,-3", Rotation="0,0,0", Scale="1,1,1") {
    rotationMode=BMM_Additive
    rotationSpace=BCS_BoneSpace
    scaleMode=BMM_Additive
    translationMode=BMM_Additive
    alphaBoolBlend {blendOption=Linear}
    alphaScaleBiasClamp {bInterpResult=true, interpSpeedIncreasing=5}
    boneToModify {boneName=mstr_frnt_torso_r}
    translation {x=3, z=-3}
   }
   n22 = "Transform (Modify) Bone - Bone: mstr_under_arm_r"(ComponentPose=@n21, Alpha=@n17, Translation="0,0,0", Rotation="-5,0,0", Scale="1,1,1") {
    rotationMode=BMM_Additive
    rotationSpace=BCS_BoneSpace
    scaleMode=BMM_Additive
    translationMode=BMM_Additive
    alphaBoolBlend {blendOption=Linear}
    alphaScaleBiasClamp {bInterpResult=true, interpSpeedIncreasing=5}
    boneToModify {boneName=mstr_under_arm_r}
    rotation {pitch=-5}
   }
   n23 = Get IsSprintingCrouch() -> (IsSprintingCrouch:float)
   n24 = "Transform (Modify) Bone - Bone: dyn_back_02_r"(ComponentPose=@n22, Alpha=@n23, Translation="0,0,0", Rotation="0,-20,0", Scale="1,1,1") {
    rotationMode=BMM_Additive
    rotationSpace=BCS_BoneSpace
    scaleMode=BMM_Additive
    translationMode=BMM_Additive
    alphaBoolBlend {blendOption=Linear}
    alphaScaleBiasClamp {bInterpResult=true, interpSpeedDecreasing=500, interpSpeedIncreasing=1}
    boneToModify {boneName=dyn_back_02_r}
    rotation {yaw=-20}
   }
   n25 = "Transform (Modify) Bone - Bone: dyn_back_02_l"(ComponentPose=@n24, Alpha=@n23, Translation="0,0,0", Rotation="0,-40,0", Scale="1,1,1") {
    rotationMode=BMM_Additive
    rotationSpace=BCS_BoneSpace
    scaleMode=BMM_Additive
    translationMode=BMM_Additive
    alphaBoolBlend {blendOption=Linear}
    alphaScaleBiasClamp {bInterpResult=true, interpSpeedDecreasing=500, interpSpeedIncreasing=1}
    boneToModify {boneName=dyn_back_02_l}
    rotation {yaw=-40}
   }
   n26 = "Transform (Modify) Bone - Bone: dyn_mid_back_02"(ComponentPose=@n25, Alpha=@n23, Translation="0,0,0", Rotation="0,30,0", Scale="1,1,1") {
    rotationMode=BMM_Additive
    rotationSpace=BCS_BoneSpace
    scaleMode=BMM_Additive
    translationMode=BMM_Additive
    alphaBoolBlend {blendOption=Linear}
    alphaScaleBiasClamp {bInterpResult=true, interpSpeedDecreasing=500, interpSpeedIncreasing=1}
    boneToModify {boneName=dyn_mid_back_02}
    rotation {yaw=30}
   }
   n27 = Get IsParachuteOpened() -> (IsParachuteOpened:float)
   n28 = "Transform (Modify) Bone - Bone: deform_upperarm_up_l"(ComponentPose=@n26, Alpha=@n27, Translation="3.340,2.218,-6.850", Rotation="0,0,0", Scale="1,1,1") {
    rotationMode=BMM_Additive
    rotationSpace=BCS_BoneSpace
    scaleMode=BMM_Additive
    translationMode=BMM_Additive
    alphaBoolBlend {blendOption=Linear}
    alphaScaleBiasClamp {bInterpResult=true, interpSpeedIncreasing=50}
    boneToModify {boneName=deform_upperarm_up_l}
    translation=<3.34 2.218 -6.85>
   }
   n29 = "Transform (Modify) Bone - Bone: deform_upperarm_up_r"(ComponentPose=@n28, Alpha=@n27, Translation="-1.210,0.319,-3.711", Rotation="0,0,0", Scale="1,1,1") {
    rotationMode=BMM_Additive
    rotationSpace=BCS_BoneSpace
    scaleMode=BMM_Additive
    translationMode=BMM_Additive
    alphaBoolBlend {blendOption=Linear}
    alphaScaleBiasClamp {bInterpResult=true, interpSpeedIncreasing=50}
    boneToModify {boneName=deform_upperarm_up_r}
    translation=<-1.21 0.319 -3.711>
   }
   n30 = Get IsSkyDiving() -> (IsSkyDiving:float)
   n31 = "Transform (Modify) Bone - Bone: dyn_back_02_r"(ComponentPose=@n29, Alpha=@n30, Translation="0,0,0", Rotation="0,10,0", Scale="1,1,1") {
    rotationMode=BMM_Additive
    rotationSpace=BCS_BoneSpace
    scaleMode=BMM_Additive
    translationMode=BMM_Additive
    alphaBoolBlend {blendOption=Linear}
    alphaScaleBiasClamp {bInterpResult=true}
    boneToModify {boneName=dyn_back_02_r}
    rotation {yaw=10}
   }
   n32 = "Transform (Modify) Bone - Bone: dyn_back_02_l"(ComponentPose=@n31, Alpha=@n30, Translation="0,0,0", Rotation="0,10,0", Scale="1,1,1") {
    rotationMode=BMM_Additive
    rotationSpace=BCS_BoneSpace
    scaleMode=BMM_Additive
    translationMode=BMM_Additive
    alphaBoolBlend {blendOption=Linear}
    alphaScaleBiasClamp {bInterpResult=true}
    boneToModify {boneName=dyn_back_02_l}
    rotation {yaw=10}
   }
   n33 = "Transform (Modify) Bone - Bone: dyn_mid_back_02"(ComponentPose=@n32, Alpha=@n30, Translation="0,0,0", Rotation="0,-10,0", Scale="1,1,1") {
    rotationMode=BMM_Additive
    rotationSpace=BCS_BoneSpace
    scaleMode=BMM_Additive
    translationMode=BMM_Additive
    alphaBoolBlend {blendOption=Linear}
    alphaScaleBiasClamp {bInterpResult=true}
    boneToModify {boneName=dyn_mid_back_02}
    rotation {yaw=-10}
   }
   n34 = Get LeftArmRot() -> (LeftArmRot:float)
   n35 = "Transform (Modify) Bone - Bone: upperarm_twist_02_r"(ComponentPose=@n33, Alpha=@n34, Translation="0,0,0", Rotation="0,0,0", Scale="1.000,1.000,0.75") {
    rotationMode=BMM_Additive
    rotationSpace=BCS_BoneSpace
    scaleMode=BMM_Additive
    translationMode=BMM_Additive
    alphaBoolBlend {blendOption=Linear}
    alphaScaleBiasClamp {bInterpResult=true, interpSpeedIncreasing=50}
    boneToModify {boneName=upperarm_twist_02_r}
    scale {z=0.75}
   }
   n36 = "Transform (Modify) Bone - Bone: deform_upperarm_up_r"(ComponentPose=@n35, Alpha=@n34, Translation="-1.210,0.319,-3.711", Rotation="0,0,0", Scale="1,1,1") {
    rotationMode=BMM_Additive
    rotationSpace=BCS_BoneSpace
    scaleMode=BMM_Additive
    translationMode=BMM_Additive
    alphaBoolBlend {blendOption=Linear}
    alphaScaleBiasClamp {bInterpResult=true, interpSpeedIncreasing=50}
    boneToModify {boneName=deform_upperarm_up_r}
    translation=<-1.21 0.319 -3.711>
   }
   n37 = Get NotMovingHeadRot() -> (NotMovingHeadRot:float)
   n38 = "Transform (Modify) Bone - Bone: mstr_back_r"(ComponentPose=@n36, Alpha=@n37, Translation="0,0,0", Rotation="0,-10,0", Scale="1,1,1") {
    rotationMode=BMM_Additive
    rotationSpace=BCS_BoneSpace
    scaleMode=BMM_Additive
    translationMode=BMM_Additive
    alphaBoolBlend {blendOption=Linear}
    alphaScaleBiasClamp {bInterpResult=true, interpSpeedDecreasing=1}
    boneToModify {boneName=mstr_back_r}
    rotation {yaw=-10}
   }
   n39 = "Transform (Modify) Bone - Bone: mstr_back_l"(ComponentPose=@n38, Alpha=@n37, Translation="0,0,0", Rotation="0,-5,0", Scale="1,1,1") {
    rotationMode=BMM_Additive
    rotationSpace=BCS_BoneSpace
    scaleMode=BMM_Additive
    translationMode=BMM_Additive
    alphaBoolBlend {blendOption=Linear}
    alphaScaleBiasClamp {bInterpResult=true, interpSpeedDecreasing=1}
    boneToModify {boneName=mstr_back_l}
    rotation {yaw=-5}
   }
   n40 = "Transform (Modify) Bone - Bone: mstr_mid_back"(ComponentPose=@n39, Alpha=@n37, Translation="0,0,0", Rotation="0,-5,0", Scale="1,1,1") {
    rotationMode=BMM_Additive
    rotationSpace=BCS_BoneSpace
    scaleMode=BMM_Additive
    translationMode=BMM_Additive
    alphaBoolBlend {blendOption=Linear}
    alphaScaleBiasClamp {bInterpResult=true, interpSpeedDecreasing=1}
    boneToModify {boneName=mstr_mid_back}
    rotation {yaw=-5}
   }
   n41 = Get IsBackpedaling() -> (IsBackpedaling:float)
   n42 = "Transform (Modify) Bone - Bone: mstr_back_r"(ComponentPose=@n40, Alpha=@n41, Translation="0,0,0", Rotation="0,-10,0", Scale="1,1,1") {
    rotationMode=BMM_Additive
    rotationSpace=BCS_BoneSpace
    scaleMode=BMM_Additive
    translationMode=BMM_Additive
    alphaBoolBlend {blendOption=Linear}
    alphaScaleBiasClamp {bInterpResult=true, interpSpeedDecreasing=1}
    boneToModify {boneName=mstr_back_r}
    rotation {yaw=-10}
   }
   n43 = "Transform (Modify) Bone - Bone: mstr_back_l"(ComponentPose=@n42, Alpha=@n41, Translation="0,0,0", Rotation="0,-5,0", Scale="1,1,1") {
    rotationMode=BMM_Additive
    rotationSpace=BCS_BoneSpace
    scaleMode=BMM_Additive
    translationMode=BMM_Additive
    alphaBoolBlend {blendOption=Linear}
    alphaScaleBiasClamp {bInterpResult=true, interpSpeedDecreasing=1}
    boneToModify {boneName=mstr_back_l}
    rotation {yaw=-5}
   }
   n44 = "Transform (Modify) Bone - Bone: mstr_mid_back"(ComponentPose=@n43, Alpha=@n41, Translation="0,0,0", Rotation="0,-5,0", Scale="1,1,1") {
    rotationMode=BMM_Additive
    rotationSpace=BCS_BoneSpace
    scaleMode=BMM_Additive
    translationMode=BMM_Additive
    alphaBoolBlend {blendOption=Linear}
    alphaScaleBiasClamp {bInterpResult=true, interpSpeedDecreasing=1}
    boneToModify {boneName=mstr_mid_back}
    rotation {yaw=-5}
   }
   n45 = Get IsCrouchingNoMovingThighUpL()
   n46 = "Transform (Modify) Bone - Bone: dyn_torso_02_l"(ComponentPose=@n44, Alpha=@n45, Translation="0,0,0", Rotation="0,0,-70", Scale="0.5,1,1") {
    rotationMode=BMM_Additive
    rotationSpace=BCS_BoneSpace
    scaleMode=BMM_Additive
    translationMode=BMM_Additive
    alphaBoolBlend {blendOption=Linear}
    alphaScaleBiasClamp {bInterpResult=true}
    boneToModify {boneName=dyn_torso_02_l}
    rotation {roll=-70}
    scale {x=0.5}
   }
   n47 = "Transform (Modify) Bone - Bone: dyn_under_arm_01_l"(ComponentPose=@n46, Alpha=@n45, Translation="0,0,0", Rotation="0,0,0", Scale="0.75,1,1") {
    rotationMode=BMM_Additive
    rotationSpace=BCS_BoneSpace
    scaleMode=BMM_Additive
    translationMode=BMM_Additive
    alphaBoolBlend {blendOption=Linear}
    alphaScaleBiasClamp {bInterpResult=true}
    boneToModify {boneName=dyn_under_arm_01_l}
    scale {x=0.75}
   }
   n48 = "Transform (Modify) Bone - Bone: dyn_torso_03_l"(
    ComponentPose=@n47
    Alpha=Get IsCrouchingIsMovingThighUpL() -> (IsCrouchingIsMovingThighUpL:float)
    Translation="0,0,0"
    Rotation="-20,0,0"
    Scale="1,1,1"
   ) {
    rotationMode=BMM_Additive
    rotationSpace=BCS_BoneSpace
    scaleMode=BMM_Additive
    translationMode=BMM_Additive
    alphaBoolBlend {blendOption=Linear}
    alphaScaleBiasClamp {bInterpResult=true}
    boneToModify {boneName=dyn_torso_03_l}
    rotation {pitch=-20}
   }
   n49 = "Transform (Modify) Bone - Bone: dyn_under_arm_03_l"(ComponentPose=@n48, Alpha=1, Translation="0,0,0", Rotation="0,-10,0", Scale="1,1,1") {
    rotationMode=BMM_Additive
    rotationSpace=BCS_BoneSpace
    scaleMode=BMM_Additive
    translationMode=BMM_Additive
    alphaBoolBlend {blendOption=Linear}
    alphaScaleBiasClamp {bInterpResult=true}
    boneToModify {boneName=dyn_under_arm_03_l}
    rotation {yaw=-10}
   }
   n50 = "Transform (Modify) Bone - Bone: dyn_torso_03_l"(ComponentPose=@n49, Alpha=1, Translation="0,0,0", Rotation="-10,0,0", Scale="1,1,1") {
    rotationMode=BMM_Additive
    rotationSpace=BCS_BoneSpace
    scaleMode=BMM_Additive
    translationMode=BMM_Additive
    alphaBoolBlend {blendOption=Linear}
    alphaScaleBiasClamp {bInterpResult=true}
    boneToModify {boneName=dyn_torso_03_l}
    rotation {pitch=-10}
   }
   n51 = "Transform (Modify) Bone - Bone: spine_02_closing_up_jacket"(
    ComponentPose=@n50
    Alpha=Get IsNotSkydiving() -> (IsNotSkydiving:float)
    Translation="-0.000,8.199,-4.049"
    Rotation="0,0,0"
    Scale="1.000,1.000,0.5"
   ) {
    rotationMode=BMM_Additive
    rotationSpace=BCS_BoneSpace
    scaleMode=BMM_Additive
    translationMode=BMM_Additive
    alphaBoolBlend {blendOption=Linear}
    alphaScaleBiasClamp {bInterpResult=true}
    boneToModify {boneName=spine_02_closing_up_jacket}
    scale {z=0.5}
    translation {y=8.199, z=-4.049}
   }
   n52 = Get CrouchingUpperArmRotZ() -> (CrouchingUpperArmRotZ:float)
   n53 = "Transform (Modify) Bone - Bone: mstr_frnt_torso_r"(ComponentPose=@n51, Alpha=@n52, Translation="-4,-2.112,-0.906", Rotation="-20,15,-1.700", Scale="1,1,1") {
    rotationMode=BMM_Additive
    rotationSpace=BCS_BoneSpace
    scaleMode=BMM_Additive
    translationMode=BMM_Additive
    alphaBoolBlend {blendOption=Linear}
    alphaScaleBiasClamp {bInterpResult=true}
    boneToModify {boneName=mstr_frnt_torso_r}
    rotation=<rot -20 15 -1.7>
    translation=<-4 -2.112 -0.906>
   }
   n54 = "Transform (Modify) Bone - Bone: mstr_under_arm_r"(ComponentPose=@n53, Alpha=@n52, Translation="-2,-2.314,-0.103", Rotation="-14.252,15,-1.700", Scale="1,1,1") {
    rotationMode=BMM_Additive
    rotationSpace=BCS_BoneSpace
    scaleMode=BMM_Additive
    translationMode=BMM_Additive
    alphaBoolBlend {blendOption=Linear}
    alphaScaleBiasClamp {bInterpResult=true}
    boneToModify {boneName=mstr_under_arm_r}
    rotation=<rot -14.252 15 -1.7>
    translation=<-2 -2.314 -0.103>
   }
   n55 = Get IsJumping() -> (IsJumping:float)
   n56 = "Transform (Modify) Bone - Bone: mstr_under_arm_r"(ComponentPose=@n54, Alpha=@n55, Translation="5,0,0", Rotation="0,0,0", Scale="1,1,1") {
    rotationMode=BMM_Additive
    rotationSpace=BCS_BoneSpace
    scaleMode=BMM_Additive
    translationMode=BMM_Additive
    alphaBoolBlend {blendOption=Linear}
    alphaScaleBiasClamp {bInterpResult=true, interpSpeedDecreasing=0.5}
    boneToModify {boneName=mstr_under_arm_r}
    translation {x=5}
   }
   n57 = "Transform (Modify) Bone - Bone: dyn_torso_01_r"(ComponentPose=@n56, Alpha=@n55, Translation="7,0,0", Rotation="-15,0,0", Scale="1,1,1") {
    rotationMode=BMM_Additive
    rotationSpace=BCS_BoneSpace
    scaleMode=BMM_Additive
    translationMode=BMM_Additive
    alphaBoolBlend {blendOption=Linear}
    alphaScaleBiasClamp {bInterpResult=true, interpSpeedDecreasing=0.5, interpSpeedIncreasing=2}
    boneToModify {boneName=dyn_torso_01_r}
    rotation {pitch=-15}
    translation {x=7}
   }
   n58 = Component To Local(ComponentPose=@n57)
   n59 = Control Rig(Source=Copy Pose From Mesh() {bCopyCurves=true, bUseAttachedParent=true}) {
    alphaBoolBlend {blendOption=Linear}
    controlRigAssetReference {blueprintRigClass=&M_MED_Canary_DeformRig_Master_ControlRig1_C}
   }
   n60 = "Transform (Modify) Bone - Bone: dyn_torso_00_l"(ComponentPose=Local To Component(LocalPose=@n59), Alpha=1, Translation="0,0,0", Rotation="15,0,0", Scale="1,1,1") {
    rotationMode=BMM_Additive
    rotationSpace=BCS_BoneSpace
    scaleMode=BMM_Additive
    translationMode=BMM_Additive
    alphaBoolBlend {blendOption=Linear}
    alphaScaleBiasClamp {bInterpResult=true, interpSpeedIncreasing=50}
    boneToModify {boneName=dyn_torso_00_l}
    rotation {pitch=15}
   }
   n61 = "Transform (Modify) Bone - Bone: dyn_torso_00_r"(ComponentPose=@n60, Alpha=1, Translation="0,0,0", Rotation="15,0,0", Scale="1,1,1") {
    rotationMode=BMM_Additive
    rotationSpace=BCS_BoneSpace
    scaleMode=BMM_Additive
    translationMode=BMM_Additive
    alphaBoolBlend {blendOption=Linear}
    alphaScaleBiasClamp {bInterpResult=true, interpSpeedIncreasing=50}
    boneToModify {boneName=dyn_torso_00_r}
    rotation {pitch=15}
   }
   n62 = "Constraint - pistol (PPPP)"(ComponentPose=@n61, Alpha=1, ConstraintWeights_0=0, ConstraintWeights_1=0, ConstraintWeights_2=1, ConstraintWeights_3=0) {
    alphaBoolBlend {blendOption=Linear}
    boneToModify {boneName=pistol}
    constraintSetup [
     {targetBone {boneName=pelvis}}
     {targetBone {boneName=thigh_r}}
     {targetBone {boneName=thigh_twist_01_r}}
     {targetBone {boneName=deform_groin_r}}
    ]
    constraintWeights [1,1,1,1]
   }
   n63 = "Transform (Modify) Bone - Bone: pistol"(ComponentPose=@n62, Alpha=1, Translation="0,0,0", Rotation="0,0,20", Scale="1,1,1") {
    rotationMode=BMM_Additive
    rotationSpace=BCS_BoneSpace
    scaleMode=BMM_Additive
    translationMode=BMM_Additive
    alphaBoolBlend {blendOption=Linear}
    alphaScaleBiasClamp {bInterpResult=true, interpSpeedIncreasing=50}
    boneToModify {boneName=pistol}
    rotation {roll=20}
   }
   n64 = "Transform (Modify) Bone - Bone: dyn_torso_00_l"(ComponentPose=@n63, Alpha=Get IsCrouchingNoMovingThighUpL(), Translation="3,2,3", Rotation="20,-20,0", Scale="0.25,1,1") {
    rotationMode=BMM_Additive
    rotationSpace=BCS_BoneSpace
    scaleMode=BMM_Additive
    translationMode=BMM_Additive
    alphaBoolBlend {blendOption=Linear}
    alphaScaleBiasClamp {interpSpeedIncreasing=1000}
    boneToModify {boneName=dyn_torso_00_l}
    rotation {pitch=20, yaw=-20}
    scale {x=0.25}
    translation=<3 2 3>
   }
   n65 = Get ThighUpR()
   n66 = "Transform (Modify) Bone - Bone: mstr_under_arm_r"(ComponentPose=@n64, Alpha=@n65, Translation="0,0,0", Rotation="5,-8,0", Scale="1,1,1") {
    rotationMode=BMM_Additive
    rotationSpace=BCS_BoneSpace
    scaleMode=BMM_Additive
    translationMode=BMM_Additive
    alphaBoolBlend {blendOption=Linear}
    alphaScaleBiasClamp {bInterpResult=true, interpSpeedIncreasing=50}
    boneToModify {boneName=mstr_under_arm_r}
    rotation {pitch=5, yaw=-8}
   }
   n67 = "Transform (Modify) Bone - Bone: mstr_frnt_torso_r"(ComponentPose=@n66, Alpha=@n65, Translation="0,0,0", Rotation="5,-8,0", Scale="1,1,1") {
    rotationMode=BMM_Additive
    rotationSpace=BCS_BoneSpace
    scaleMode=BMM_Additive
    translationMode=BMM_Additive
    alphaBoolBlend {blendOption=Linear}
    alphaScaleBiasClamp {bInterpResult=true, interpSpeedIncreasing=50}
    boneToModify {boneName=mstr_frnt_torso_r}
    rotation {pitch=5, yaw=-8}
   }
   n68 = RigidBody(ComponentPose=@n67, OverrideWorldGravity=Get GravityOverride()) {
    alphaInputType=Bool
    bClampLinearTranslationLimitToRefPose=true
    bOverrideWorldGravity=true
    overridePhysicsAsset=&/M_MED_Canary_JacketPhysicsBB
    simulationSpace=BaseBoneSpace
    alphaBoolBlend {blendOption=Linear}
    baseBoneRef {boneName=spine_05}
    componentAppliedLinearAccClamp=<500 500 500>
    componentLinearAccScale=<0.15 0.25 0.25>
    componentLinearVelScale=<0.3 0.3 0.3>
    overrideWorldGravity {z=-980}
    simSpaceSettings {
     maxAngularAcceleration=0
     maxAngularVelocity=0.3
     maxLinearAcceleration=100
     maxLinearVelocity=100
     velocityScaleZ=0
     worldAlpha=1
     externalLinearDragV=<0.2 0.2>
    }
   }
   n69 = Get IsNotCrouching()
   n70 = "Transform (Modify) Bone - Bone: mstr_frnt_torso_r"(ComponentPose=@n68, Alpha=@n69, Translation="-2,0,2", Rotation="3,0,0", Scale="1,1,1") {
    rotationMode=BMM_Additive
    rotationSpace=BCS_BoneSpace
    scaleMode=BMM_Additive
    translationMode=BMM_Additive
    alphaBoolBlend {blendOption=Linear}
    alphaScaleBiasClamp {bInterpResult=true, interpSpeedIncreasing=50}
    boneToModify {boneName=mstr_frnt_torso_r}
    rotation {pitch=3}
    translation {x=-2, z=2}
   }
   n71 = "Transform (Modify) Bone - Bone: mstr_under_arm_r"(ComponentPose=@n70, Alpha=@n69, Translation="-2,0,2", Rotation="3,0,0", Scale="1,1,1") {
    rotationMode=BMM_Additive
    rotationSpace=BCS_BoneSpace
    scaleMode=BMM_Additive
    translationMode=BMM_Additive
    alphaBoolBlend {blendOption=Linear}
    alphaScaleBiasClamp {bInterpResult=true, interpSpeedIncreasing=50}
    boneToModify {boneName=mstr_under_arm_r}
    rotation {pitch=3}
    translation {x=-2, z=2}
   }
   n72 = Get IsCrouching()
   n73 = "Transform (Modify) Bone - Bone: mstr_frnt_torso_r"(ComponentPose=@n71, Alpha=@n72, Translation="-1,0,1", Rotation="3,0,0", Scale="1,1,1") {
    rotationMode=BMM_Additive
    rotationSpace=BCS_BoneSpace
    scaleMode=BMM_Additive
    translationMode=BMM_Additive
    alphaBoolBlend {blendOption=Linear}
    alphaScaleBiasClamp {bInterpResult=true, interpSpeedIncreasing=50}
    boneToModify {boneName=mstr_frnt_torso_r}
    rotation {pitch=3}
    translation {x=-1, z=1}
   }
   n74 = "Transform (Modify) Bone - Bone: mstr_under_arm_r"(ComponentPose=@n73, Alpha=@n72, Translation="-1,0,1", Rotation="3,0,0", Scale="1,1,1") {
    rotationMode=BMM_Additive
    rotationSpace=BCS_BoneSpace
    scaleMode=BMM_Additive
    translationMode=BMM_Additive
    alphaBoolBlend {blendOption=Linear}
    alphaScaleBiasClamp {bInterpResult=true, interpSpeedIncreasing=50}
    boneToModify {boneName=mstr_under_arm_r}
    rotation {pitch=3}
    translation {x=-1, z=1}
   }
   n75 = Get IsSprintingNoCrouch()
   n76 = "Transform (Modify) Bone - Bone: dyn_back_02_r"(ComponentPose=@n74, Alpha=@n75, Translation="0,0,0", Rotation="0,0,0", Scale="1,1,1") {
    rotationMode=BMM_Additive
    rotationSpace=BCS_BoneSpace
    scaleMode=BMM_Additive
    translationMode=BMM_Additive
    alphaBoolBlend {blendOption=Linear}
    alphaScaleBiasClamp {bInterpResult=true, interpSpeedDecreasing=500, interpSpeedIncreasing=1}
    boneToModify {boneName=dyn_back_02_r}
   }
   n77 = "Transform (Modify) Bone - Bone: dyn_back_02_l"(ComponentPose=@n76, Alpha=@n75, Translation="0,0,0", Rotation="0,-10,0", Scale="1,1,1") {
    rotationMode=BMM_Additive
    rotationSpace=BCS_BoneSpace
    scaleMode=BMM_Additive
    translationMode=BMM_Additive
    alphaBoolBlend {blendOption=Linear}
    alphaScaleBiasClamp {bInterpResult=true, interpSpeedDecreasing=500, interpSpeedIncreasing=1}
    boneToModify {boneName=dyn_back_02_l}
    rotation {yaw=-10}
   }
   n78 = "Transform (Modify) Bone - Bone: dyn_mid_back_02"(ComponentPose=@n77, Alpha=@n75, Translation="0,0,0", Rotation="0,10,0", Scale="1,1,1") {
    rotationMode=BMM_Additive
    rotationSpace=BCS_BoneSpace
    scaleMode=BMM_Additive
    translationMode=BMM_Additive
    alphaBoolBlend {blendOption=Linear}
    alphaScaleBiasClamp {bInterpResult=true, interpSpeedDecreasing=500, interpSpeedIncreasing=1}
    boneToModify {boneName=dyn_mid_back_02}
    rotation {yaw=10}
   }
   n79 = "Transform (Modify) Bone - Bone: mstr_frnt_torso_r"(ComponentPose=@n78, Alpha=@n75, Translation="3,0,-3", Rotation="0,0,0", Scale="1,1,1") {
    rotationMode=BMM_Additive
    rotationSpace=BCS_BoneSpace
    scaleMode=BMM_Additive
    translationMode=BMM_Additive
    alphaBoolBlend {blendOption=Linear}
    alphaScaleBiasClamp {bInterpResult=true, interpSpeedIncreasing=5}
    boneToModify {boneName=mstr_frnt_torso_r}
    translation {x=3, z=-3}
   }
   n80 = "Transform (Modify) Bone - Bone: mstr_under_arm_r"(ComponentPose=@n79, Alpha=@n75, Translation="0,0,0", Rotation="-5,0,0", Scale="1,1,1") {
    rotationMode=BMM_Additive
    rotationSpace=BCS_BoneSpace
    scaleMode=BMM_Additive
    translationMode=BMM_Additive
    alphaBoolBlend {blendOption=Linear}
    alphaScaleBiasClamp {bInterpResult=true, interpSpeedIncreasing=5}
    boneToModify {boneName=mstr_under_arm_r}
    rotation {pitch=-5}
   }
   n81 = Get IsSprintingCrouch()
   n82 = "Transform (Modify) Bone - Bone: dyn_back_02_r"(ComponentPose=@n80, Alpha=@n81, Translation="0,0,0", Rotation="0,-20,0", Scale="1,1,1") {
    rotationMode=BMM_Additive
    rotationSpace=BCS_BoneSpace
    scaleMode=BMM_Additive
    translationMode=BMM_Additive
    alphaBoolBlend {blendOption=Linear}
    alphaScaleBiasClamp {bInterpResult=true, interpSpeedDecreasing=500, interpSpeedIncreasing=1}
    boneToModify {boneName=dyn_back_02_r}
    rotation {yaw=-20}
   }
   n83 = "Transform (Modify) Bone - Bone: dyn_back_02_l"(ComponentPose=@n82, Alpha=@n81, Translation="0,0,0", Rotation="0,-40,0", Scale="1,1,1") {
    rotationMode=BMM_Additive
    rotationSpace=BCS_BoneSpace
    scaleMode=BMM_Additive
    translationMode=BMM_Additive
    alphaBoolBlend {blendOption=Linear}
    alphaScaleBiasClamp {bInterpResult=true, interpSpeedDecreasing=500, interpSpeedIncreasing=1}
    boneToModify {boneName=dyn_back_02_l}
    rotation {yaw=-40}
   }
   n84 = "Transform (Modify) Bone - Bone: dyn_mid_back_02"(ComponentPose=@n83, Alpha=@n81, Translation="0,0,0", Rotation="0,30,0", Scale="1,1,1") {
    rotationMode=BMM_Additive
    rotationSpace=BCS_BoneSpace
    scaleMode=BMM_Additive
    translationMode=BMM_Additive
    alphaBoolBlend {blendOption=Linear}
    alphaScaleBiasClamp {bInterpResult=true, interpSpeedDecreasing=500, interpSpeedIncreasing=1}
    boneToModify {boneName=dyn_mid_back_02}
    rotation {yaw=30}
   }
   n85 = Get IsParachuteOpened()
   n86 = "Transform (Modify) Bone - Bone: deform_upperarm_up_l"(ComponentPose=@n84, Alpha=@n85, Translation="3.340,2.218,-6.850", Rotation="0,0,0", Scale="1,1,1") {
    rotationMode=BMM_Additive
    rotationSpace=BCS_BoneSpace
    scaleMode=BMM_Additive
    translationMode=BMM_Additive
    alphaBoolBlend {blendOption=Linear}
    alphaScaleBiasClamp {bInterpResult=true, interpSpeedIncreasing=50}
    boneToModify {boneName=deform_upperarm_up_l}
    translation=<3.34 2.218 -6.85>
   }
   n87 = "Transform (Modify) Bone - Bone: deform_upperarm_up_r"(ComponentPose=@n86, Alpha=@n85, Translation="-1.210,0.319,-3.711", Rotation="0,0,0", Scale="1,1,1") {
    rotationMode=BMM_Additive
    rotationSpace=BCS_BoneSpace
    scaleMode=BMM_Additive
    translationMode=BMM_Additive
    alphaBoolBlend {blendOption=Linear}
    alphaScaleBiasClamp {bInterpResult=true, interpSpeedIncreasing=50}
    boneToModify {boneName=deform_upperarm_up_r}
    translation=<-1.21 0.319 -3.711>
   }
   n88 = Get IsSkyDiving()
   n89 = "Transform (Modify) Bone - Bone: dyn_back_02_r"(ComponentPose=@n87, Alpha=@n88, Translation="0,0,0", Rotation="0,10,0", Scale="1,1,1") {
    rotationMode=BMM_Additive
    rotationSpace=BCS_BoneSpace
    scaleMode=BMM_Additive
    translationMode=BMM_Additive
    alphaBoolBlend {blendOption=Linear}
    alphaScaleBiasClamp {bInterpResult=true}
    boneToModify {boneName=dyn_back_02_r}
    rotation {yaw=10}
   }
   n90 = "Transform (Modify) Bone - Bone: dyn_back_02_l"(ComponentPose=@n89, Alpha=@n88, Translation="0,0,0", Rotation="0,10,0", Scale="1,1,1") {
    rotationMode=BMM_Additive
    rotationSpace=BCS_BoneSpace
    scaleMode=BMM_Additive
    translationMode=BMM_Additive
    alphaBoolBlend {blendOption=Linear}
    alphaScaleBiasClamp {bInterpResult=true}
    boneToModify {boneName=dyn_back_02_l}
    rotation {yaw=10}
   }
   n91 = "Transform (Modify) Bone - Bone: dyn_mid_back_02"(ComponentPose=@n90, Alpha=@n88, Translation="0,0,0", Rotation="0,-10,0", Scale="1,1,1") {
    rotationMode=BMM_Additive
    rotationSpace=BCS_BoneSpace
    scaleMode=BMM_Additive
    translationMode=BMM_Additive
    alphaBoolBlend {blendOption=Linear}
    alphaScaleBiasClamp {bInterpResult=true}
    boneToModify {boneName=dyn_mid_back_02}
    rotation {yaw=-10}
   }
   n92 = Get LeftArmRot()
   n93 = "Transform (Modify) Bone - Bone: upperarm_twist_02_r"(ComponentPose=@n91, Alpha=@n92, Translation="0,0,0", Rotation="0,0,0", Scale="1.000,1.000,0.75") {
    rotationMode=BMM_Additive
    rotationSpace=BCS_BoneSpace
    scaleMode=BMM_Additive
    translationMode=BMM_Additive
    alphaBoolBlend {blendOption=Linear}
    alphaScaleBiasClamp {bInterpResult=true, interpSpeedIncreasing=50}
    boneToModify {boneName=upperarm_twist_02_r}
    scale {z=0.75}
   }
   n94 = "Transform (Modify) Bone - Bone: deform_upperarm_up_r"(ComponentPose=@n93, Alpha=@n92, Translation="-1.210,0.319,-3.711", Rotation="0,0,0", Scale="1,1,1") {
    rotationMode=BMM_Additive
    rotationSpace=BCS_BoneSpace
    scaleMode=BMM_Additive
    translationMode=BMM_Additive
    alphaBoolBlend {blendOption=Linear}
    alphaScaleBiasClamp {bInterpResult=true, interpSpeedIncreasing=50}
    boneToModify {boneName=deform_upperarm_up_r}
    translation=<-1.21 0.319 -3.711>
   }
   n95 = Get NotMovingHeadRot()
   n96 = "Transform (Modify) Bone - Bone: mstr_back_r"(ComponentPose=@n94, Alpha=@n95, Translation="0,0,0", Rotation="0,-10,0", Scale="1,1,1") {
    rotationMode=BMM_Additive
    rotationSpace=BCS_BoneSpace
    scaleMode=BMM_Additive
    translationMode=BMM_Additive
    alphaBoolBlend {blendOption=Linear}
    alphaScaleBiasClamp {bInterpResult=true, interpSpeedDecreasing=1}
    boneToModify {boneName=mstr_back_r}
    rotation {yaw=-10}
   }
   n97 = "Transform (Modify) Bone - Bone: mstr_back_l"(ComponentPose=@n96, Alpha=@n95, Translation="0,0,0", Rotation="0,-5,0", Scale="1,1,1") {
    rotationMode=BMM_Additive
    rotationSpace=BCS_BoneSpace
    scaleMode=BMM_Additive
    translationMode=BMM_Additive
    alphaBoolBlend {blendOption=Linear}
    alphaScaleBiasClamp {bInterpResult=true, interpSpeedDecreasing=1}
    boneToModify {boneName=mstr_back_l}
    rotation {yaw=-5}
   }
   n98 = "Transform (Modify) Bone - Bone: mstr_mid_back"(ComponentPose=@n97, Alpha=@n95, Translation="0,0,0", Rotation="0,-5,0", Scale="1,1,1") {
    rotationMode=BMM_Additive
    rotationSpace=BCS_BoneSpace
    scaleMode=BMM_Additive
    translationMode=BMM_Additive
    alphaBoolBlend {blendOption=Linear}
    alphaScaleBiasClamp {bInterpResult=true, interpSpeedDecreasing=1}
    boneToModify {boneName=mstr_mid_back}
    rotation {yaw=-5}
   }
   n99 = Get IsBackpedaling()
   n100 = "Transform (Modify) Bone - Bone: mstr_back_r"(ComponentPose=@n98, Alpha=@n99, Translation="0,0,0", Rotation="0,-10,0", Scale="1,1,1") {
    rotationMode=BMM_Additive
    rotationSpace=BCS_BoneSpace
    scaleMode=BMM_Additive
    translationMode=BMM_Additive
    alphaBoolBlend {blendOption=Linear}
    alphaScaleBiasClamp {bInterpResult=true, interpSpeedDecreasing=1}
    boneToModify {boneName=mstr_back_r}
    rotation {yaw=-10}
   }
   n101 = "Transform (Modify) Bone - Bone: mstr_back_l"(ComponentPose=@n100, Alpha=@n99, Translation="0,0,0", Rotation="0,-5,0", Scale="1,1,1") {
    rotationMode=BMM_Additive
    rotationSpace=BCS_BoneSpace
    scaleMode=BMM_Additive
    translationMode=BMM_Additive
    alphaBoolBlend {blendOption=Linear}
    alphaScaleBiasClamp {bInterpResult=true, interpSpeedDecreasing=1}
    boneToModify {boneName=mstr_back_l}
    rotation {yaw=-5}
   }
   n102 = "Transform (Modify) Bone - Bone: mstr_mid_back"(ComponentPose=@n101, Alpha=@n99, Translation="0,0,0", Rotation="0,-5,0", Scale="1,1,1") {
    rotationMode=BMM_Additive
    rotationSpace=BCS_BoneSpace
    scaleMode=BMM_Additive
    translationMode=BMM_Additive
    alphaBoolBlend {blendOption=Linear}
    alphaScaleBiasClamp {bInterpResult=true, interpSpeedDecreasing=1}
    boneToModify {boneName=mstr_mid_back}
    rotation {yaw=-5}
   }
   n103 = Get IsCrouchingNoMovingThighUpL()
   n104 = "Transform (Modify) Bone - Bone: dyn_torso_02_l"(ComponentPose=@n102, Alpha=@n103, Translation="0,0,0", Rotation="0,0,-70", Scale="0.5,1,1") {
    rotationMode=BMM_Additive
    rotationSpace=BCS_BoneSpace
    scaleMode=BMM_Additive
    translationMode=BMM_Additive
    alphaBoolBlend {blendOption=Linear}
    alphaScaleBiasClamp {bInterpResult=true}
    boneToModify {boneName=dyn_torso_02_l}
    rotation {roll=-70}
    scale {x=0.5}
   }
   n105 = "Transform (Modify) Bone - Bone: dyn_under_arm_01_l"(ComponentPose=@n104, Alpha=@n103, Translation="0,0,0", Rotation="0,0,0", Scale="0.75,1,1") {
    rotationMode=BMM_Additive
    rotationSpace=BCS_BoneSpace
    scaleMode=BMM_Additive
    translationMode=BMM_Additive
    alphaBoolBlend {blendOption=Linear}
    alphaScaleBiasClamp {bInterpResult=true}
    boneToModify {boneName=dyn_under_arm_01_l}
    scale {x=0.75}
   }
   n106 = "Transform (Modify) Bone - Bone: dyn_torso_03_l"(ComponentPose=@n105, Alpha=Get IsCrouchingIsMovingThighUpL(), Translation="0,0,0", Rotation="-20,0,0", Scale="1,1,1") {
    rotationMode=BMM_Additive
    rotationSpace=BCS_BoneSpace
    scaleMode=BMM_Additive
    translationMode=BMM_Additive
    alphaBoolBlend {blendOption=Linear}
    alphaScaleBiasClamp {bInterpResult=true}
    boneToModify {boneName=dyn_torso_03_l}
    rotation {pitch=-20}
   }
   n107 = "Transform (Modify) Bone - Bone: dyn_under_arm_03_l"(ComponentPose=@n106, Alpha=1, Translation="0,0,0", Rotation="0,-10,0", Scale="1,1,1") {
    rotationMode=BMM_Additive
    rotationSpace=BCS_BoneSpace
    scaleMode=BMM_Additive
    translationMode=BMM_Additive
    alphaBoolBlend {blendOption=Linear}
    alphaScaleBiasClamp {bInterpResult=true}
    boneToModify {boneName=dyn_under_arm_03_l}
    rotation {yaw=-10}
   }
   n108 = "Transform (Modify) Bone - Bone: dyn_torso_03_l"(ComponentPose=@n107, Alpha=1, Translation="0,0,0", Rotation="-10,0,0", Scale="1,1,1") {
    rotationMode=BMM_Additive
    rotationSpace=BCS_BoneSpace
    scaleMode=BMM_Additive
    translationMode=BMM_Additive
    alphaBoolBlend {blendOption=Linear}
    alphaScaleBiasClamp {bInterpResult=true}
    boneToModify {boneName=dyn_torso_03_l}
    rotation {pitch=-10}
   }
   n109 = "Transform (Modify) Bone - Bone: spine_02_closing_up_jacket"(ComponentPose=@n108, Alpha=Get IsNotSkydiving(), Translation="-0.000,8.199,-4.049", Rotation="0,0,0", Scale="1.000,1.000,0.5") {
    rotationMode=BMM_Additive
    rotationSpace=BCS_BoneSpace
    scaleMode=BMM_Additive
    translationMode=BMM_Additive
    alphaBoolBlend {blendOption=Linear}
    alphaScaleBiasClamp {bInterpResult=true}
    boneToModify {boneName=spine_02_closing_up_jacket}
    scale {z=0.5}
    translation {y=8.199, z=-4.049}
   }
   n110 = Get CrouchingUpperArmRotZ()
   n111 = "Transform (Modify) Bone - Bone: mstr_frnt_torso_r"(ComponentPose=@n109, Alpha=@n110, Translation="-4,-2.112,-0.906", Rotation="-20,15,-1.700", Scale="1,1,1") {
    rotationMode=BMM_Additive
    rotationSpace=BCS_BoneSpace
    scaleMode=BMM_Additive
    translationMode=BMM_Additive
    alphaBoolBlend {blendOption=Linear}
    alphaScaleBiasClamp {bInterpResult=true}
    boneToModify {boneName=mstr_frnt_torso_r}
    rotation=<rot -20 15 -1.7>
    translation=<-4 -2.112 -0.906>
   }
   n112 = "Transform (Modify) Bone - Bone: mstr_under_arm_r"(ComponentPose=@n111, Alpha=@n110, Translation="-2,-2.314,-0.103", Rotation="-14.252,15,-1.700", Scale="1,1,1") {
    rotationMode=BMM_Additive
    rotationSpace=BCS_BoneSpace
    scaleMode=BMM_Additive
    translationMode=BMM_Additive
    alphaBoolBlend {blendOption=Linear}
    alphaScaleBiasClamp {bInterpResult=true}
    boneToModify {boneName=mstr_under_arm_r}
    rotation=<rot -14.252 15 -1.7>
    translation=<-2 -2.314 -0.103>
   }
   n113 = Component To Local(ComponentPose=@n112)
   n114 = Get bIsWearingBackbling() -> (bIsWearingBackbling:bool)
   n115 = RigidBody(
    ComponentPose=Local To Component(LocalPose=Blend Poses by bool(BlendPose_0=@n113, BlendPose_1=@n58, BlendTime_0=0.3, BlendTime_1=0.3, bActiveValue=@n114) {blendType=Linear})
    OverrideWorldGravity=Get GravityOverride()
   ) {
    alphaInputType=Bool
    bClampLinearTranslationLimitToRefPose=true
    bOverrideWorldGravity=true
    overridePhysicsAsset=&/M_MED_Canary_RopeBagPhysics
    simulationSpace=BaseBoneSpace
    alphaBoolBlend {blendOption=Linear}
    baseBoneRef {boneName=pelvis}
    componentAppliedLinearAccClamp=<500 500 500>
    componentLinearAccScale=<0.5 0.3 0.3>
    componentLinearVelScale=<0.3 0.3 0.3>
    overrideWorldGravity {z=-980}
    simSpaceSettings {
     maxAngularAcceleration=100
     maxAngularVelocity=0.5
     maxLinearAcceleration=200
     maxLinearVelocity=200
     velocityScaleZ=0
     worldAlpha=1
     externalLinearDragV=<0.2 0.2>
    }
   }
   n116 = "Transform (Modify) Bone - Bone: mstr_ropeMain"(ComponentPose=@n115, Alpha=1, Translation="0,0,0", Rotation="0,5,0", Scale="1,1,1") {
    rotationMode=BMM_Additive
    rotationSpace=BCS_BoneSpace
    scaleMode=BMM_Additive
    translationMode=BMM_Additive
    alphaBoolBlend {blendOption=Linear}
    alphaScaleBiasClamp {bInterpResult=true, interpSpeedIncreasing=50}
    boneToModify {boneName=mstr_ropeMain}
    rotation {yaw=5}
   }
   n117 = "Transform (Modify) Bone - Bone: mstr_ropeMain"(ComponentPose=@n116, Alpha=Get IsCrouchingNoMovingThighUpL(), Translation="2,0,0", Rotation="0,0,-5", Scale="1,1,1") {
    rotationMode=BMM_Additive
    rotationSpace=BCS_BoneSpace
    scaleMode=BMM_Additive
    translationMode=BMM_Additive
    alphaBoolBlend {blendOption=Linear}
    alphaScaleBiasClamp {bInterpResult=true}
    boneToModify {boneName=mstr_ropeMain}
    rotation {roll=-5}
    translation {x=2}
   }
   n118 = "Transform (Modify) Bone - Bone: mstr_ropeMain"(ComponentPose=@n117, Alpha=Get IsCrouchingMoving2D() -> (IsCrouchingMoving2D:float), Translation="4,0,0", Rotation="0,0,-15", Scale="1,1,1") {
    rotationMode=BMM_Additive
    rotationSpace=BCS_BoneSpace
    scaleMode=BMM_Additive
    translationMode=BMM_Additive
    alphaBoolBlend {blendOption=Linear}
    alphaScaleBiasClamp {bInterpResult=true}
    boneToModify {boneName=mstr_ropeMain}
    rotation {roll=-15}
    translation {x=4}
   }
   n119 = Get IsInFrontEndNotEmoting() -> (IsInFrontEndNotEmoting:float)
   n120 = "Transform (Modify) Bone - Bone: dyn_torso_00_l"(ComponentPose=@n118, Alpha=@n119, Translation="0,0,0", Rotation="-13,0,0", Scale="1,1,1") {
    rotationMode=BMM_Additive
    rotationSpace=BCS_BoneSpace
    scaleMode=BMM_Additive
    translationMode=BMM_Additive
    alphaBoolBlend {blendOption=Linear}
    alphaScaleBiasClamp {bInterpResult=true, interpSpeedDecreasing=20, interpSpeedIncreasing=5}
    boneToModify {boneName=dyn_torso_00_l}
    rotation {pitch=-13}
   }
   n121 = "Transform (Modify) Bone - Bone: dyn_torso_00_r"(ComponentPose=@n120, Alpha=@n119, Translation="0,0,0", Rotation="-10,0,0", Scale="1,1,1") {
    rotationMode=BMM_Additive
    rotationSpace=BCS_BoneSpace
    scaleMode=BMM_Additive
    translationMode=BMM_Additive
    alphaBoolBlend {blendOption=Linear}
    alphaScaleBiasClamp {bInterpResult=true, interpSpeedDecreasing=20, interpSpeedIncreasing=5}
    boneToModify {boneName=dyn_torso_00_r}
    rotation {pitch=-10}
   }
   n122 = "Transform (Modify) Bone - Bone: mstr_under_arm_r"(ComponentPose=@n121, Alpha=@n119, Translation="0,0,0", Rotation="-10,0,0", Scale="1,1,1") {
    rotationMode=BMM_Additive
    rotationSpace=BCS_BoneSpace
    scaleMode=BMM_Additive
    translationMode=BMM_Additive
    alphaBoolBlend {blendOption=Linear}
    alphaScaleBiasClamp {bInterpResult=true, interpSpeedDecreasing=20, interpSpeedIncreasing=5}
    boneToModify {boneName=mstr_under_arm_r}
    rotation {pitch=-10}
   }
   n123 = "Transform (Modify) Bone - Bone: mstr_bag_Main"(ComponentPose=@n122, Alpha=@n119, Translation="0,0,0", Rotation="0,-15,0", Scale="1,1,1") {
    rotationMode=BMM_Additive
    rotationSpace=BCS_BoneSpace
    scaleMode=BMM_Additive
    translationMode=BMM_Additive
    alphaBoolBlend {blendOption=Linear}
    alphaScaleBiasClamp {bInterpResult=true, interpSpeedDecreasing=20, interpSpeedIncreasing=5}
    boneToModify {boneName=mstr_bag_Main}
    rotation {yaw=-15}
   }
   n124 = "Transform (Modify) Bone - Bone: dyn_bag_Main2"(ComponentPose=@n123, Alpha=@n119, Translation="0,0,0", Rotation="0,15,0", Scale="1,1,1") {
    rotationMode=BMM_Additive
    rotationSpace=BCS_BoneSpace
    scaleMode=BMM_Additive
    translationMode=BMM_Additive
    alphaBoolBlend {blendOption=Linear}
    alphaScaleBiasClamp {bInterpResult=true, interpSpeedDecreasing=20, interpSpeedIncreasing=5}
    boneToModify {boneName=dyn_bag_Main2}
    rotation {yaw=15}
   }
   n125 = "Transform (Modify) Bone - Bone: mstr_ropeMain"(ComponentPose=@n124, Alpha=@n119, Translation="0,-1,0", Rotation="0,-15,0", Scale="1,1,1") {
    rotationMode=BMM_Additive
    rotationSpace=BCS_BoneSpace
    scaleMode=BMM_Additive
    translationMode=BMM_Additive
    alphaBoolBlend {blendOption=Linear}
    alphaScaleBiasClamp {bInterpResult=true, interpSpeedDecreasing=20, interpSpeedIncreasing=5}
    boneToModify {boneName=mstr_ropeMain}
    rotation {yaw=-15}
    translation {y=-1}
   }
   n126 = "Transform (Modify) Bone - Bone: mstr_under_arm_l"(ComponentPose=@n125, Alpha=@n119, Translation="0,0,0", Rotation="-5,0,0", Scale="1,1,1") {
    rotationMode=BMM_Additive
    rotationSpace=BCS_BoneSpace
    scaleMode=BMM_Additive
    translationMode=BMM_Additive
    alphaBoolBlend {blendOption=Linear}
    alphaScaleBiasClamp {bInterpResult=true, interpSpeedDecreasing=20, interpSpeedIncreasing=5}
    boneToModify {boneName=mstr_under_arm_l}
    rotation {pitch=-5}
   }
   n127 = "Transform (Modify) Bone - Bone: dyn_torso_00_l"(
    ComponentPose=@n126
    Alpha=Get IsInFrontEndNoBackbling() -> (IsInFrontEndNoBackbling:float)
    Translation="0,0,0"
    Rotation="5,0,0"
    Scale="1,1,1"
   ) {
    rotationMode=BMM_Additive
    rotationSpace=BCS_BoneSpace
    scaleMode=BMM_Additive
    translationMode=BMM_Additive
    alphaBoolBlend {blendOption=Linear}
    alphaScaleBiasClamp {bInterpResult=true, interpSpeedDecreasing=20, interpSpeedIncreasing=5}
    boneToModify {boneName=dyn_torso_00_l}
    rotation {pitch=5}
   }
   Output Pose(Result=Component To Local(ComponentPose=@n127)) {layerGroup=None}
  ]}
 EventGraph {nodes [
   Event BlueprintUpdateAnimation()
   > Is Valid(InputObject=GetOwnerSkeletalMeshComponent())
   > n1 = Gravity Override State Function DEPRECATED(
    CharacterPartAnimInstance=Self-Reference()
    GameTimeInSeconds=GetGameTimeInSeconds()
    GravityOverrideParams=Get GravityOverrideParams() -> (GravityOverrideParams:Gravity Override Params Struct Structure)
   )
   > Set GravityOverride(GravityOverride=@n1)
   > n2 = PlayerMovement(AnimInstance=Self-Reference(), XDotProductCoefficient=1, YDotProductCoefficient=1, ZDotProductCoefficient=1)
   > Set PlayerMovementFunction_X(PlayerMovementFunction_X:float=@n2) -> (Output_Get:float)
   > Set PlayerMovementFunction_Y(PlayerMovementFunction_Y:float=@n2.YMovement) -> (Output_Get:float)
  ]}
]
variables [GravityOverrideParams {type=GravityOverrideParamsStruct}]
