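# Post-copy Animation Blueprint for the Canary character's face accessory that nudges the hat bone as the head rotates back.
# Each update maps the head's twist relative to spine_05 to an alpha, then multiplies it by a not-skydiving factor to disable it while skydiving.
# The AnimGraph copies the attached parent's pose, then uses that alpha to blend an additive bone-space translation and yaw onto the hat.
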AnimBlueprint /BRCosmetics/Characters/Player/Male/Medium/Bodies/M_MED_Canary/Meshes/Parts/M_MED_Canary_FaceAcc_AnimBP
target_skeleton=&/M_MED_Canary_FaceAcc_Skeleton
graphs [
 AnimGraph {nodes [
   n1 = "Transform (Modify) Bone - Bone: hat"(
    ComponentPose=Local To Component(LocalPose=Copy Pose From Mesh() {bUseAttachedParent=true})
    Alpha=Get headRotBackNoSkydive() -> (headRotBackNoSkydive:double)
    Translation="3,0,0"
    Rotation="0,10,0"
    Scale="1,1,1"
   ) {
    rotationMode=BMM_Additive
    rotationSpace=BCS_BoneSpace
    scaleMode=BMM_Replace
    scaleSpace=BCS_BoneSpace
    translationMode=BMM_Additive
    translationSpace=BCS_BoneSpace
    alphaBoolBlend {blendOption=Linear}
    boneToModify {boneName=hat}
    rotation {yaw=10}
    translation {x=3}
   }
   Output Pose(Result=Component To Local(ComponentPose=@n1)) {layerGroup=None}
  ]}
 EventGraph {nodes [
   Event BlueprintUpdateAnimation()
   > Set headRotBack(headRotBack:double=GetAlphaFromReferenceTransformDelta(
     SkeletalMeshComponent=GetOwnerSkeletalMeshComponent()
     BoneName=head
     BaseName=spine_05
     InputRange=(X=-50,Y=0)
     OutputRange=(X=1,Y=0)
     TwistAxis="1,0,0"
    )) -> (Output_Get:double)
   n1 = Get bIsSkyDiving()
   > Set isNotSkydiving(isNotSkydiving:double="To Float (Boolean)"(InBool=NOT Boolean(A=@n1))) -> (Output_Get:double)
   > Set headRotBackNoSkydive(headRotBackNoSkydive:double=float * float(A=Get headRotBack() -> (headRotBack:double), B=Get isNotSkydiving() -> (isNotSkydiving:double)) {is=title K2Node_PromotableOperator}) -> (Output_Get:double)
  ]}
]
variables [
 headRotBack {type=number}
 headRotBackNoSkydive {type=number}
 isNotSkydiving {type=number}
]
